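# Material function that adds wind flutter to Mask Maker paper through World Position Offset and normals, behind a static switch that otherwise passes the input attributes through.
# Layered sines built on local-Z pivot gradients and per-instance random time shape the sway, with vertex color red masking movement.
# Two panning Perlin normal samples on local XZ coordinates, read at a fixed mip, are angle-corrected blended into noise and ripple normals.
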MaterialFunction /Hermes/Techart/LivingWorld/MF_WindPaperMaskMaker
graph {nodes [
  n1 = Divide(
   A=ComponentMask(None=LocalPosition() {includedOffsets=ExcludeOffsets}.XYZ) {b=true, r=true}
   B=ScalarParameter() {defaultValue=1050.4492, group=Mask Maker Paper, parameterName=UV}
  ) {constB=125}
  n2 = Multiply(A=NamedRerouteUsage(), B=ScalarParameter() {defaultValue=0.2, group=Mask Maker Paper, parameterName=Speed Max, sliderMax=1}) {constB=0.0001}
  n3 = Constant() {r=2}
  n4 = TextureSample!(UVs=Add(A=@n1, B=Divide(A=@n2) {constB=3}), Tex, Level=@n3, Apply View MipBias) {mipValueMode=TMVM_MipLevel, samplerType=SAMPLERTYPE_Normal, texture=&/T_WPO_Perlin_Normal}
  n5 = Multiply(A=@n1) {constB=2}
  n6 = ComponentMask(None=@n5) {r=true}
  n7 = Divide(A=@n2) {constB=-3.3}
  n8 = ComponentMask(None=@n5) {g=true}
  n9 = Divide(A=@n2) {constB=3.3}
  n10 = TextureSample(UVs=AppendVector(A=Add(A=@n6, B=@n7), B=Add(A=@n8, B=@n9)), Level=@n3) {mipValueMode=TMVM_MipLevel, samplerType=SAMPLERTYPE_Normal, texture=&T_WPO_Perlin_Normal}
  n11 = MaterialFunctionCall(BaseNormal=@n4.RGB, AdditionalNormal=@n10.RGB) {materialFunction=&BlendAngleCorrectedNormals}
  n12 = ComponentMask(None=@n11.Result) {g=true, r=true}
  n13 = AppendVector(A=@n12, B=Constant())
  n14 = Multiply(A=@n13, B=NamedRerouteUsage())
  Add(A=@n14) {constB=0.5}
  Add(A=@n14) {constB=-0.05}
  Add()
  Desaturation(None=@n12)
  n15 = NamedRerouteUsage()
  n16 = Add(A=NamedRerouteUsage(), B=NamedRerouteUsage())
  n17 = NamedRerouteUsage()
  FunctionOutput(None=StaticSwitchParameter(
    True=SetMaterialAttributes!(MaterialAttributes=@n15, WorldPositionOffset=@n16, EmissiveColor, Normal=@n17) {attributeSetTypes [#1, #2, #3]}
    False=@n15
   ) {group=Mask Maker Paper, parameterName=Mask Maker Paper?})
  n18 = Multiply(A=NamedRerouteUsage()) {constB=-1}
  n19 = Constant() {r=0.63}
  n20 = Multiply(A=Add(A=NamedRerouteUsage(), B=Multiply(A=@n18, B=@n19))) {constB=5}
  LinearInterpolate(A=Constant3Vector() {constant=<rgba 0 0 -1 0>}, B=Constant3Vector() {constant=<rgba 0 0 1 0>}, Alpha=ConstantBiasScale(None=Sine(None=@n20)))
  n21 = Add(A=NamedRerouteUsage()) {constB=0.4}
  n22 = NamedRerouteUsage()
  n23 = Add(A=OneMinus(None=NamedRerouteUsage())) {constB=0.1}
  n24 = NamedRerouteUsage()
  n25 = Add(A=Multiply(A=Multiply(A=@n21, B=@n22), B=Multiply(A=@n23, B=@n24)), B=@n21)
  n26 = Multiply(A=@n25, B=@n25)
  MaterialFunctionCall!(X range Y range, Gradient Height, ShowGrid?, Thickness, Color, "f(x)"=@n26) {materialFunction=/Engine/Functions/Engine_MaterialFunctions01/Debug/PlotFunctionOnGraph}
  n27 = ScalarParameter() {defaultValue=2, parameterName=Height Noise}
  n28 = Normalize(VectorInput=Add(A=Multiply(A=@n13, B=@n27) {constB=3}, B=NamedRerouteUsage()))
  MaterialFunctionCall("f(x)"=@n28) {materialFunction=&PlotFunctionOnGraph}
  n29 = NamedRerouteUsage()
  n30 = NamedRerouteUsage()
  n31 = Sine(None=Multiply(
    A=Add(A=@n29, B=@n30)
    B=ScalarParameter() {defaultValue=2.4, group=Mask Maker Paper, parameterName=Wooble Wave Frequency 2, sliderMax=1, sortPriority=4}
   ))
  n32 = Multiply(A=VertexColor().R, B=ScalarParameter() {defaultValue=1, group=Mask Maker Paper, parameterName=Vertex Color})
  n33 = NamedRerouteDeclaration(None=Multiply(A=Multiply(A=@n31) {constB=2, desc="Intensity wave "}, B=Saturate(None=@n32))) {name=Movement Mask, nodeColor=<rgba 0.5 1 0 1>}
  Multiply(A=@n33, B=NamedRerouteUsage())
  n34 = GetMaterialAttributes(None=FunctionInput() {inputType=FunctionInput_MaterialAttributes}) {attributeGetTypes [
    #1
    #3
    #2
    #4
   ]}
  NamedRerouteDeclaration(None=@n34.Normal) {name=OG Normal, nodeColor=<rgba 0 0.0523 1 1>}
  NamedRerouteDeclaration(None=@n34.EmissiveColor) {name=Emissive, nodeColor=<rgba 1 0.4922 0 1>}
  n35 = Constant3Vector() {constant=<rgba 0 1 0 0>}
  n36 = Sign(None=Normalize(VectorInput=Transform(None=@n35) {transformSourceType=TRANSFORMSOURCE_Instance}))
  n37 = OneMinus(None=@n36)
  n38 = Add(A=NamedRerouteUsage(), B=NamedRerouteUsage())
  n39 = Add(A=Sine(None=Multiply(A=@n38) {constB=0.37}))
  n40 = Add(A=NamedRerouteUsage()) {constB=0.1}
  n41 = LinearInterpolate(A=@n36, B=Add(A=@n36, B=@n37), Alpha=Multiply(A=Add(A=@n39, B=@n40)) {constB=0.4})
  n42 = Multiply(A=Add(A=NamedRerouteUsage(), B=NamedRerouteUsage())) {constB=0.47}
  n43 = Add(A=Add(A=Sine(None=@n42)), B=NamedRerouteUsage())
  n44 = Multiply(A=@n43) {constB=0.2}
  NamedRerouteDeclaration(None=LinearInterpolate(A=@n41, B=LinearInterpolate(A=@n41, B=@n37) {constAlpha=0.1}, Alpha=Saturate(None=@n44))) {name=Foward Vector, nodeColor=<rgba 0.4766 1 0 1>}
  n45 = MaterialFunctionCall() {materialFunction=&ObjectPivotPoint}
  n46 = ComponentMask(None=Subtract(
    A=LocalPosition() {includedOffsets=ExcludeOffsets}.XYZ
    B=Transform(None=@n45.Object Pivot Location) {transformSourceType=TRANSFORMSOURCE_World, transformType=TRANSFORM_Local}
   )) {b=true}
  n47 = NamedRerouteDeclaration(None=Divide(A=@n46, B=ScalarParameter() {defaultValue=1000, group=Mask Maker Paper, parameterName=Divide 4})) {name=Gradient_4, nodeColor=<rgba 0 0.5078 1 1>}
  n48 = NamedRerouteUsage()
  n49 = Sine(None=Multiply(A=Add(A=@n47, B=@n48)) {constB=0.77})
  NamedRerouteDeclaration(None=Multiply(A=Add(A=@n49)) {constB=0.5}) {name=Noise, nodeColor=<rgba 0.5703 0 1 1>}
  NamedRerouteDeclaration(None=Divide(A=@n46, B=ScalarParameter() {defaultValue=1550, group=Mask Maker Paper, parameterName=Divide 2})) {name=Gradient_2, nodeColor=<rgba 0 0.5078 1 1>}
  NamedRerouteDeclaration(None=Divide(A=@n46, B=ScalarParameter() {defaultValue=1400, group=Mask Maker Paper, parameterName=Divide 3})) {name=Gradient_3, nodeColor=<rgba 0 0.5078 1 1>}
  NamedRerouteDeclaration(None=Add(A=@n4.R, B=@n10.G)) {name=Noise normal, nodeColor=<rgba 0 0.25 1 1>}
  NamedRerouteDeclaration(None=@n43) {name=Noise 2, nodeColor=<rgba 1 0 0.8437 1>}
  n50 = PerInstanceRandom()
  NamedRerouteDeclaration(None=Add(A=LinearInterpolate(Alpha=@n50) {constA=0.37, constB=56763}, B=Time())) {name=Random time, nodeColor=<rgba 0 1 0.7422 1>}
  NamedRerouteDeclaration(None=@n34.BaseColor) {name=OG Color, nodeColor=<rgba 1 0.4453 0 1>}
  NamedRerouteDeclaration(None=@n34.MaterialAttributes) {name=MA OG, nodeColor=<rgba 1 0 0.8906 1>}
  NamedRerouteDeclaration(None=@n34.WorldPositionOffset) {name=WPO OG, nodeColor=<rgba 0 0.8125 1 1>}
  n51 = NamedRerouteUsage()
  n52 = Multiply(A=LinearInterpolate(Alpha=Saturate(None=@n26)) {constA=0.5, constB=1.8}, B=@n33)
  NamedRerouteDeclaration(None=Multiply(
    A=Multiply(A=@n51, B=@n52)
    B=ScalarParameter() {defaultValue=30, group=WindChains, parameterName=Wooble Intensity, sliderMax=1, sortPriority=4}
   )) {name=New WPO, nodeColor=<rgba 1 0.0499 0.0595 1>}
  NamedRerouteDeclaration(None=Divide(A=@n46, B=ScalarParameter() {defaultValue=600, group=Mask Maker Paper, parameterName=Divide})) {name=Gradient, nodeColor=<rgba 0 0.5078 1 1>}
  NamedRerouteDeclaration(None=@n28) {name=Height, nodeColor=<rgba 1 0 0.0469 1>}
  n53 = NamedRerouteUsage()
  n54 = Add(
   A=@n53
   B=MaterialFunctionCall!(InNormal=@n11.Result, Strength=ScalarParameter() {defaultValue=1.5, group=Mask Maker Paper, parameterName=Normal Intensity}) {materialFunction=/Game/Effects/Fort_Effects/FrontEnd/S33/GateContent/SilenceMoon_NormalStrength}.Result
  )
  n55 = NamedRerouteUsage()
  NamedRerouteDeclaration(None=LinearInterpolate(A=@n53, B=Normalize(VectorInput=@n54), Alpha=Saturate(None=@n55))) {name=Mixed Norm, nodeColor=<rgba 0 0.3672 1 1>}
  NamedRerouteUsage()
  NamedRerouteUsage()
  NamedRerouteUsage()
  NamedRerouteUsage()
  Power()
  Time()
  Transform(None=@n11.Result) {transformType=TRANSFORM_Local}
 ]}
