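# Opaque Nanite and instancing-ready material for Mystic ancient mask paper props that sway through World Position Offset.
# Vertices rotate about a pivot on an axis blended by a sine, with destructible damage layered through MF_FN_DestructibleDamage.
# WindChains helpers build position-seeded time, height gradient, two-layer panning Perlin wobble and normals, and UV1 surface masks.
# NaniteReplace supplies a custom-primitive-data pivot for Nanite, actor position otherwise.
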Material /Hermes/Techart/LivingWorld/MM_MaskPaper_Animated
graph {
 properties {
  bAlwaysEvaluateWorldPositionOffset=true
  bAutomaticallySetUsageInEditor=false
  bUseMaterialAttributes=true
  bUsedWithInstancedStaticMeshes=true
  bUsedWithNanite=true
  maxWorldPositionOffsetDisplacement=200
  usedShadingModels=Default Lit
 }
 nodes [
  n1 = NamedRerouteUsage()
  n2 = MaterialFunctionCall(Float3=@n1) {materialFunction=&BreakOutFloat3Components}
  n3 = ScalarParameter() {defaultValue=199.1234, group=WindChains, parameterName=Movement Intensity Gradient, sortPriority=3}
  MaterialFunctionCall(X=@n2.R, Y=@n2.R, Z=Subtract(A=@n2.B, B=@n3) {constB=500}) {materialFunction=&MakeFloat3}
  MaterialFunctionCall() {materialFunction=&FixRotateAboutAxisNormals}
  Multiply()
  n4 = Divide(A=WorldPosition().XY, B=ScalarParameter() {defaultValue=1565.9993, group=WindChains, parameterName=UV}) {constB=125}
  n5 = Multiply(A=NamedRerouteUsage(), B=ScalarParameter() {defaultValue=0.6, group=WindChains, parameterName=Speed Max, sliderMax=1}) {constB=0.0001}
  n6 = Add(A=@n4, B=Divide(A=@n5) {constB=3})
  n7 = TextureSample(UVs=@n6) {texture=&/T_WPO_Perlin}
  n8 = Multiply(A=@n4) {constB=2}
  n9 = AppendVector(
   A=Add(A=ComponentMask(None=@n8) {r=true}, B=Divide(A=@n5) {constB=-3.3})
   B=Add(A=ComponentMask(None=@n8) {g=true}, B=Divide(A=@n5) {constB=3.3})
  )
  n10 = TextureSample(UVs=@n9) {texture=&T_WPO_Perlin}
  n11 = Add(A=Multiply(A=Multiply(A=@n7.R, B=@n10.R)) {constB=2}) {constB=-0.5}
  n12 = Multiply(
   A=ScalarParameter() {defaultValue=5, group=WindChains, parameterName=Wobloe Intensity, sortPriority=4}
   B=NamedRerouteDeclaration(None=Multiply(A=@n11) {constB=2}) {name=Height, nodeColor=<rgba 1 0 0.0469 1>}
  ) {constB=5}
  Multiply(A=@n12, B=LinearInterpolate(Alpha=VertexColor().B) {constA=1, constB=1.1})
  n13 = Ceil(None=TransformPosition(None=Constant3Vector()) {transformType=TRANSFORMPOSSOURCE_World})
  n14 = Constant() {desc="LeapOfFaith ", r=10000}
  n15 = MaterialFunctionCall(Float3=Multiply(A=Divide(A=@n13, B=@n14)) {constB=11.17}) {materialFunction=&BreakOutFloat3Components}
  n16 = Frac(None=Add(A=@n15.R, B=@n15.R))
  NamedRerouteDeclaration(None=Add(A=Multiply(A=@n16) {constB=120}, B=Time())) {name=Random time, nodeColor=<rgba 0 1 0.7422 1>}
  n17 = Add(A=ComponentMask(None=TextureCoordinate() {coordinateIndex=1}) {r=true}) {constB=-0.5}
  n18 = Round(None=Abs(None=@n17))
  n19 = Multiply(A=OneMinus(None=Saturate(None=@n18)), B=ScalarParameter() {group=WindChains, parameterName=Projector L1, sortPriority=-4})
  n20 = Add(A=@n17) {constB=-1}
  n21 = Round(None=Abs(None=@n20))
  n22 = Multiply(A=OneMinus(None=Saturate(None=@n21)), B=ScalarParameter() {defaultValue=1, group=WindChains, parameterName=Projector L2, sortPriority=-3})
  n23 = Add(A=@n20) {constB=-1}
  n24 = Round(None=Abs(None=@n23))
  n25 = Multiply(A=OneMinus(None=Saturate(None=@n24)), B=ScalarParameter() {group=WindChains, parameterName=Projector L3, sortPriority=-2})
  n26 = Round(None=Abs(None=Add(A=@n23) {constB=-1}))
  n27 = Multiply(A=OneMinus(None=Saturate(None=@n26)), B=ScalarParameter() {group=WindChains, parameterName=Projector L4, sortPriority=-1})
  NamedRerouteDeclaration(None=Max(A=Max(A=@n19, B=@n22), B=Max(A=@n25, B=@n27))) {name=Surface Mask, nodeColor=<rgba 1 0 0.8438 1>}
  NamedRerouteDeclaration(None=TransformPosition(None=LocalPosition().XYZ) {transformSourceType=TRANSFORMPOSSOURCE_Instance, transformType=TRANSFORMPOSSOURCE_TranslatedWorld}) {name=Actor Position, nodeColor=<rgba 1 0 0.0545 1>}
  NamedRerouteDeclaration(None=Transform(None=TextureSample() {samplerType=SAMPLERTYPE_Normal, texture=/Hermes/Props/Mystic/Mystic_AncientMaskPaper_A/Textures/T_Mystic_AncientMaskPaper_A_A_N}.RGB)) {name=WS Normal, nodeColor=<rgba 0 0.0523 1 1>}
  n28 = Divide(A=Subtract(A=WorldPosition() {worldPositionShaderOffset=WPT_CameraRelativeNoOffsets}.Z, B=ComponentMask(None=@n1) {b=true}), B=@n3) {constB=500}
  NamedRerouteDeclaration(None=Clamp(None=OneMinus(None=@n28)) {clampMode=CMODE_ClampMin}) {name=Gradient, nodeColor=<rgba 0.4766 1 0 1>}
  NamedRerouteDeclaration(None=@n12) {name=Woble, nodeColor=<rgba 0 0.8125 1 1>}
  NamedRerouteUsage()
  n29 = VectorParameter() {bUseCustomPrimitiveData=true, parameterName=PivotCord, channelNames {a="", b=Z, g=Y, r=X}, defaultValue=<rgba 0 0 0 1>}
  NaniteReplace!(
   Default=ActorPositionWS() {originType=CameraRelative}
   Nanite=TransformPosition(None=@n29) {transformType=TRANSFORMPOSSOURCE_TranslatedWorld}
  )
  n30 = MaterialFunctionCall!(
   InNormal=MaterialFunctionCall(
    BaseNormal=TextureSample(UVs=@n6) {samplerType=SAMPLERTYPE_Normal, texture=&/T_WPO_Perlin_Normal}.RGB
    AdditionalNormal=TextureSample(UVs=@n9) {samplerType=SAMPLERTYPE_Normal, texture=&T_WPO_Perlin_Normal}.RGB
   ) {materialFunction=&BlendAngleCorrectedNormals}.Result
   Strength=ScalarParameter() {defaultValue=2, group=WindChains, parameterName=Normal Intensity}
  ) {materialFunction=/Game/Effects/Fort_Effects/FrontEnd/S33/GateContent/SilenceMoon_NormalStrength}
  Normalize(VectorInput=Add(A=NamedRerouteUsage(), B=Transform(None=@n30.Result)))
  n31 = TextureSample() {samplerType=SAMPLERTYPE_Masks, texture=/Hermes/Props/Mystic/Mystic_AncientMaskPaper_A/Textures/T_Mystic_AncientMaskPaper_A_A_S}
  n32 = Constant3Vector() {constant=<rgba 1 0 0 0>}
  n33 = NamedRerouteUsage()
  n34 = Multiply(
   A=Add(A=ScalarParameter() {defaultValue=0.5, group=WindChains, parameterName=Swing Intensity, sliderMax=1, sortPriority=2}) {constB=0.1}
   B=ScalarParameter() {defaultValue=0.1, group=WindChains, parameterName=Swing Max RandomDirection Speed, sliderMax=0.5, sortPriority=2}
  ) {constB=0.1}
  n35 = ConstantBiasScale(None=Sine(None=Multiply(A=@n33, B=@n34))) {desc=Direction Alpha}
  n36 = Constant3Vector() {constant=<rgba -1 0 0 0>}
  n37 = Constant3Vector() {constant=<rgba 0 1 0 0>}
  n38 = MaterialFunctionCall!(
   RotationAngle=ScalarParameter() {defaultValue=0.9947, parameterName=Rotaion Angle}
   NormalizedRotationAxis=Normalize(VectorInput=MaterialFunctionCall(1=@n32, 2=@n37, 3=@n36, A=@n35) {desc=Rotation Axis Lerp, materialFunction=&Lerp_Multiple_Float3}.Lerp 3 Inputs)
   PivotPoint=NamedRerouteUsage()
   Position=WorldPosition() {worldPositionShaderOffset=WPT_CameraRelative}.XYZ
   Normal=NamedRerouteUsage()
   TangentU
   TangentV
  ) {materialFunction=/Game/Characters/Player/Male/Medium/Bodies/M_MED_Inferno/FX/Mat/RotateVertexAboutAxis}
  n39 = SetMaterialAttributes!(
   MaterialAttributes
   Base Color=TextureSample() {texture=/Hermes/Props/Mystic/Mystic_AncientMaskPaper_A/Textures/T_Mystic_AncientMaskPaper_A_A_D}.RGB
   Metallic=@n31.G
   Specular=@n31.G
   Roughness=@n31.G
   World Position Offset=@n38.WPO
   Normal=Transform(None=@n38.Normal) {transformSourceType=TRANSFORMSOURCE_World, transformType=TRANSFORM_Tangent}
   Emissive Color=NamedRerouteUsage()
  ) {attributeSetTypes [
    #1
    #2
    #3
    #4
    #5
    #6
    #7
   ]}
  n40 = MaterialFunctionCall(Material_In=@n39, UseAdditiveWPO=StaticBool() {value=true}) {materialFunction=&MF_FN_DestructibleDamage}
  SetMaterialAttributes!(MaterialAttributes=@n40.Material_Out, World Position Offset=@n40.Material_Out) {attributeSetTypes [#5]}
  Time()
 ]
}
